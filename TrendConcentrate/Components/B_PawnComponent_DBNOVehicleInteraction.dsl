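; Pawn component that lets a downed (DBNO) player dropped or thrown by a carrier land in a vehicle seat, and applies reboot van marking while downed.
; Carry tags drive it: hoisting blocks interaction and records enemy throws, dropping spawns a temporary vehicle-overlap sphere.
; Overlap force-seats the player in an enterable vehicle's nearest free non-driver seat, favoring an enabled revive seat unless enemy-thrown.
; Cleanup removes the sphere when falling or DBNO ends.
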(event EventBeginPlay
  (Utilities|FlowControl|Delay 1.0)
  (bind _asfort_player_pawn (Utilities|Casting|CastToFortPlayerPawn (Components|GetOwner)))
  (bind _output_get (Variables|Default|SetFortPawn _asfort_player_pawn))
  (EventDispatchers|BindEventtoOnDBNOStateChanged _output_get (delegate Custom|Event%20OnDBNOStateChanged))
  (Character|AssignMovementModeChangedDelegate (Variables|Default|GetFortPawn) (EventDispatchers|CreateEvent :!event Custom|MovementModeChangedDelegate_Event))
  (Ability|Tasks|WaitGameplayTagAddToActor (Variables|Default|GetFortPawn) "(TagName=\"Gameplay.Status.DBNOBeingCarried\")"
    (:then
      (Ability|Tasks|WaitGameplayTagAddToActor (Variables|Default|GetFortPawn) "(TagName=\"Gameplay.Status.DBNOBeingDropped\")"
        (:Added
          ; Remove Interact Blocks when Dropped
          (GameplayEffects|RemoveActiveGameplayEffect (Ability|GetAbilitySystemComponent (Variables|Default|GetFortPawn)) (Variables|Default|GetBlockInteractHandle) -1)
          (bind _asfort_player_pawn_2 (Utilities|Casting|CastToFortPlayerPawn (Components|GetOwner)))
          (bind _returnvalue (Components|AddDynamicComponentbyClass _asfort_player_pawn_2 "/Script/Engine.SphereComponent"))
          (bind _assphere_collision (Utilities|Casting|CastToSphereComponent _returnvalue))
          (Variables|Default|SetDBNOOverlapComponent _assphere_collision)
          (Collision|SetCollisionObjectType (Variables|Default|GetDBNOOverlapComponent) "ECC_WorldDynamic")
          (Collision|SetCollisionEnabled (Variables|Default|GetDBNOOverlapComponent) "QueryOnly")
          (Collision|SetCollisionResponsetoChannel (Variables|Default|GetDBNOOverlapComponent) "ECC_Vehicle" "ECR_Overlap")
          (Variables|/Script/Engine.PrimitiveComponent|SetbGenerateOverlapEvents true (Variables|Default|GetDBNOOverlapComponent))
          (Collision|AssignOnComponentBeginOverlap (Variables|Default|GetDBNOOverlapComponent) (EventDispatchers|CreateEvent :!event Custom|OnComponentBeginOverlap_Event))
          (Components|Sphere|SetSphereRadius (Variables|Default|GetDBNOOverlapComponent) (ScalableFloat|GetValueatLevel (Variables|Default|GetOverlap%20Radius)) true))))
    (:Added
      ; Block Interacts While Hoisted
      (bind _returnvalue_2 (GameplayEffects|ApplyGameplayEffectToSelf :self (Ability|GetAbilitySystemComponent (Variables|Default|GetFortPawn)) :GameplayEffectClass "/ImmunityCarriage/Gameplay/Effects/GE_ImmunityCarriage_BlockInteract.GE_ImmunityCarriage_BlockInteract_C" :Level 0.0))
      (Variables|Default|SetBlockInteractHandle _returnvalue_2)
      (bind _asfort_player_pawn_athena (Utilities|Casting|CastToFortPlayerPawnAthena (Class|FortPlayerPawn|GetDBNOHoister (Variables|Default|GetFortPawn))))
      (bind _asfort_player_pawn_athena_2 (Utilities|Casting|CastToFortPlayerPawnAthena (Variables|Default|GetFortPawn)))
      (if (Utilities|Enum|Equal(Enum) (Team|GetTeamAffiliation _asfort_player_pawn_athena_2 _asfort_player_pawn_athena) "Friendly")
        (Variables|Default|SetbThrownByEnemy false)
        (else
          (Variables|Default|SetbThrownByEnemy true))))))

(event Custom|Event%20OnDBNOStateChanged
  (params (FortPawn :type Fort%20Pawn%20Object%20Reference) (bInIsDBNO :type Boolean))
  (bind _returnvalue (Ability|GetFortAbilitySystemComponent FortPawn))
  (if bInIsDBNO
    (bind _returnvalue_2 (GameplayEffects|ApplyGameplayEffectToSelf _returnvalue "/TrendConcentrate/Gameplay/Effects/GE_TrendConcentrate_DBNO_RebootVan_Passive.GE_TrendConcentrate_DBNO_RebootVan_Passive_C" 0.0 (Variables|Default|GetEffect%20Context)))
    (Variables|Default|SetRebootVanMarkEffect _returnvalue_2)
    (Abilities|TryActivateAbilitybyClass _returnvalue "/TrendConcentrate/Gameplay/Abilities/GA_ImmunityCarriage_Mark_Passive_Parent.GA_ImmunityCarriage_Mark_Passive_Parent_C" true)
    (else
      (GameplayEffects|RemoveActiveGameplayEffect _returnvalue (Variables|Default|GetRebootVanMarkEffect) -1)
      (CallFunction|Cleanup))))

(event Custom|OnComponentBeginOverlap_Event
  (params (OverlappedComponent :type Primitive%20Component%20Object%20Reference) (OtherActor :type Actor%20Object%20Reference) (OtherComp :type Primitive%20Component%20Object%20Reference) (OtherBodyIndex :type Integer) (bFromSweep :type Boolean) (SweepResult :type Hit%20Result%20Structure%20%28by%20ref%29))
  ; Car Throw Enabled Check
  (if (== (ScalableFloat|GetValueatLevel (Variables|Default|GetDBNOCarThrowEnabled)) (Variables|Default|GetEnabled))
    (CallFunction|OnOverlap OtherActor)))

(event Custom|MovementModeChangedDelegate_Event
  (params (Character :type Character%20Object%20Reference) (PrevMovementMode :type EMovementMode%20Enum :default "MOVE_None") (PreviousCustomMode :type Byte))
  (bind _fortpawn (Variables|Default|GetFortPawn)
    (:then
      (if (Utilities|Enum|NotEqual(Enum) (Variables|/Script/Engine.CharacterMovementComponent|GetMovementMode (Variables|/Script/Engine.Character|GetCharacterMovement _fortpawn)) "MOVE_Falling")
        (CallFunction|Cleanup)))
    (:else)))

(fn CanEnterVehicle
  (params (Vehicle :type Fort%20Athena%20Vehicle%20Object%20Reference))
  (bind _asfort_player_pawn_athena (Utilities|Casting|CastToFortPlayerPawnAthena (Variables|Default|GetFortPawn)))
  (LocalVariables|SetFort%20Pawn%20Athena _asfort_player_pawn_athena)
  (return (Math|Boolean|ANDBoolean (Class|FortAthenaVehicle|HasSeatAvailableFor Vehicle (Variables|Default|GetFortPawn)) (Class|FortAthenaVehicle|IsVehicleStateSuitableforEntry Vehicle) (not (Class|FortAthenaVehicle|IsSquadRestricted Vehicle (LocalVariables|GetFort%20Pawn%20Athena))) (not (Class|FortVehicleInterface|IsTeamRestricted Vehicle (LocalVariables|GetFort%20Pawn%20Athena))))))

(fn CheckRezSeat
  (params (Vehicle :type Fort%20Athena%20Vehicle%20Object%20Reference))
  (bind _asvalet_immunity_carriage_vehicle (Utilities|Casting|CastToValet_ImmunityCarriage_Vehicle Vehicle)
    (:then
      (return (Variables|/ImmunityCarriage/Vehicle/Valet_ImmunityCarriage_Vehicle.Valet_ImmunityCarriage_Vehicle_C|GetReviveSeatIndex _asvalet_immunity_carriage_vehicle) (Class|ValetImmunityCarriageVehicle|GetIsRezSeatEnabled _asvalet_immunity_carriage_vehicle) true))
    (:CastFailed
      (return 0 false false))))

(fn Cleanup
  (bind _dbnooverlapcomponent (Variables|Default|GetDBNOOverlapComponent)
    (:then
      (Collision|UnbindallEventsfromOnComponentBeginOverlap _dbnooverlapcomponent)
      (bind _asfort_player_pawn (Utilities|Casting|CastToFortPlayerPawn (Components|GetOwner)))
      (Components|DestroyDynamicComponent _asfort_player_pawn _dbnooverlapcomponent)
      (Variables|Default|SetDBNOOverlapComponent 0))
    (:else)))

(fn OnOverlap
  (params (OtherActor :type Actor%20Object%20Reference))
  (bind _vehicle (LocalVariables|GetVehicle))
  (bind (_rezseatindex _rezseatenabled _isrebootvan) (CallFunction|CheckRezSeat _vehicle))
  (bind _output_seatsocket (Utilities|Struct|BreakAthenaCarPlayerSlot (Utilities|Array|Get(aref) (LocalVariables|GetVehicleSlots) (LocalVariables|GetFoundSeatIndex))))
  (bind _asfort_athena_vehicle (Utilities|Casting|CastToFortAthenaVehicle OtherActor))
  (if (not (Utilities|IsValid (LocalVariables|GetVehicle) :!class /Script/Engine.KismetSystemLibrary))
    (CallFunction|Cleanup)
    (bind _output_get (LocalVariables|SetVehicle _asfort_athena_vehicle))
    (bind _returnvalue (CallFunction|CanEnterVehicle _output_get))
    (if _returnvalue
      (bind _returnvalue_2 (Vehicle|GetVehicleSeatComponent (LocalVariables|GetVehicle)))
      (bind _output_get_2 (LocalVariables|SetVehicleSlots (Variables|/Script/FortniteGame.FortVehicleSeatComponent|GetPlayerSlots _returnvalue_2)))
      (bind (_array_element _array_index) (Utilities|Array|ForEachLoopwithBreak _output_get_2)
        (:LoopBody
          (if (!= _array_index 0)
            (Utilities|IsValid (Vehicle|GetPawnatSeat (LocalVariables|GetVehicle) _array_index)
              (:Is%20Not%20Valid
                ;[1 1 1] Check for rez seat, break on found rez seat because we want to force the player into it
                (if (and _isrebootvan (== _rezseatindex _array_index))
                  (if _rezseatenabled
                    (if (not (Variables|Default|GetbThrownByEnemy))
                      (LocalVariables|SetFoundSeatIndex _array_index) ; Found rez seat, break
                    )))
                (if (>= (LocalVariables|GetFoundSeatIndex) 0)
                  (bind _output_seatsocket_2 (Utilities|Struct|BreakAthenaCarPlayerSlot (Utilities|Array|Get(aref) (LocalVariables|GetVehicleSlots) _array_index)))
                  (if (< (Math|Vector|Distance(Vector) (Transformation|GetActorLocation (Variables|Default|GetFortPawn)) (Transformation|GetSocketLocation (Variables|/Script/FortniteGame.FortAthenaVehicle|GetSkeletalMesh (LocalVariables|GetVehicle)) _output_seatsocket_2)) (Math|Vector|Distance(Vector) (Transformation|GetActorLocation (Variables|Default|GetFortPawn)) (Transformation|GetSocketLocation (Variables|/Script/FortniteGame.FortAthenaVehicle|GetSkeletalMesh (LocalVariables|GetVehicle)) _output_seatsocket)))
                    (LocalVariables|SetFoundSeatIndex _array_index))
                  (else
                    (LocalVariables|SetFoundSeatIndex _array_index))))) ; Seat Is empty
          ))
        (:Completed
          _))
      (bind _asfort_player_pawn (Utilities|Casting|CastToFortPlayerPawn (Components|GetOwner)))
      ;[1 1 1] check for valid index
      (if (not (<= (LocalVariables|GetFoundSeatIndex) 0))
        (Vehicle|ForceSeatSinglePawn (LocalVariables|GetVehicle) _asfort_player_pawn (LocalVariables|GetFoundSeatIndex))))))
