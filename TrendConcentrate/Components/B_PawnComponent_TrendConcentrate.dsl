(event Custom|SoundLibraryControl
  (params (FortPawn :type Fort%20Pawn%20Object%20Reference) (IsDBNO :type Boolean))
  (bind _soundlibrarytimer (Variables|Default|GetSoundLibraryTimer))
  ; Adds a sound library for extra fall hit notify
  (Variables|Default|SetFort%20Pawn FortPawn)
  (if IsDBNO
    (bind _returnvalue (Utilities|Time|SetTimerbyEvent (EventDispatchers|CreateEvent :!event Custom|DelayedLibraryAdd) 0.5))
    (Variables|Default|SetSoundLibraryTimer _returnvalue)
    (elif (Utilities|Time|IsValidTimerHandle _soundlibrarytimer)
      (Utilities|Time|ClearandInvalidateTimerbyHandle _soundlibrarytimer)
      (else
        (if (> (FortPawn|GetHealth (Variables|Default|GetFort%20Pawn)) 0)
          (SoundLibrary|RemoveLibrary (WorldSubsystems|GetSoundLibrarySubsystem) (Variables|Default|GetFort%20Pawn) (Variables|Default|GetLibrary) :!class /Script/SoundLibrary.SoundLibrarySubsystem))))))

(event Custom|DelayedLibraryAdd
  ; Adds a sound library for extra fall hit notify
  (SoundLibrary|AddLibrary (WorldSubsystems|GetSoundLibrarySubsystem) (Variables|Default|GetFort%20Pawn) (Variables|Default|GetLibrary) :!class /Script/SoundLibrary.SoundLibrarySubsystem))

(event Custom|On%20DBNO%20State%20Changed
  (params (FortPawn :type Fort%20Pawn%20Object%20Reference) (bInIsDBNO :type Boolean))
  (bind _pawn (CallFunction|GetOwnerPawn))
  (bind _returnvalue (Actor|GetComponentbyClass _pawn "/Script/VaultingCodeRuntime.HurdleComponent"))
  (bind _pawn_2 (CallFunction|GetOwnerPawn))
  (bind _returnvalue_2 (Actor|GetComponentbyClass _pawn_2 "/Script/VaultingCodeRuntime.HurdleComponent"))
  (bind (_bsprint _broll) (CallFunction|ShouldEnableInputMappings))
  ;[1 1 1] Handle DBNO State Changes
  (CallFunction|SoundLibraryControl FortPawn bInIsDBNO)
  (if bInIsDBNO
    (bind _output_get (Variables|Default|SetbDBNOWasLocallyControlled (Pawn|IsLocallyControlled FortPawn))) ; Cache to examine later.
    (if _output_get
      ;;[1 1 1] When DBNO begins, add the input mapping context for DBNO Sprint
      (CallFunction|SetMappingContextsEnabled (CallFunction|GetOwnerPawn) _bsprint _broll true)
      ;;[1 1 1 0.5] Deplete Stamina on DBNO based on Scalable Float
      (if (ScalableFloat|GetValueasBool (Variables|Default|GetEnergyBurnOnDBNO_Enabled))
        (Energy|UseEnergy (Actor|GetComponentbyClass (CallFunction|GetOwnerPawn) "/Script/EnergyRuntime.FortComponent_Energy") (ScalableFloat|GetValueatLevel (Variables|Default|GetEnergyBurnOnDBNO_SF)) true 0)
        (else
          _))
      (Utilities|IsValid _returnvalue
        (:Is%20Valid
          (Class|HurdleComponent|PushHurdleInputConfigOverride _returnvalue (Variables|Default|GetDBNO%20Override%20Tag) (Variables|Default|GetDBNO%20Hurdle%20Input%20Config)))
        (:Is%20Not%20Valid
          _))
      ;; Checking if the pawn will ragdoll if they roll on Touch
      (if (Inputs|IsUsingInputType "Touch" (Fortnite|GetFortPlayerControllerfromActor (CallFunction|GetPawnOwner)))
        (Variables|Default|SetbPrevRagdollCheckResult false)
        (Character|BindEventtoOnCharacterMovementUpdated (CallFunction|GetPawnOwner) (EventDispatchers|CreateEvent :!event OnMovementUpdated)))
      (elif (ScalableFloat|GetValueasBool (Variables|Default|GetApplyBlockDriverTagDirectly_Enabled))
        ;;[1 1 1] Add "BlockDriverSeat" Tag
        (Ability|GameplayEffect|AddLooseGameplayTags FortPawn (GameplayTags|MakeGameplayTagContainerfromTag "(TagName=\"Athena.Gameplay.Status.BlockDriverSeat\")") true)))
    (elif (or (Pawn|IsLocallyControlled FortPawn) (Variables|Default|GetbDBNOWasLocallyControlled))
      ;;[1 1 1] When DBNO Ends, remove the Input Mapping Context for DBNO Sprint
      (CallFunction|SetMappingContextsEnabled (CallFunction|GetOwnerPawn))
      ;;[1 1 1] Reset State
      (CallFunction|EndMMEWithTags (GameplayTags|MakeGameplayTagContainerfromTag "(TagName=\"MovementMode.TrendConcentrate.Roll\")"))
      (CallFunction|EndMMEWithTags (GameplayTags|MakeGameplayTagContainerfromTag "(TagName=\"MovementMode.TrendConcentrate.Ragdoll\")"))
      (bind _showholsteredweaponcomponent (Variables|Default|GetShowHolsteredWeaponComponent)
        (:then
          (Class|FortAthenaShowHolsteredWeaponsComponent|ClearHideHolsteredWeaponFunction _showholsteredweaponcomponent))
        (:else
          _))
      ;;[1 1 1] Pop DBNO Hurdle Input Config
      (Utilities|IsValid _returnvalue_2
        (:Is%20Valid
          (Class|HurdleComponent|PopHurdleInputConfigOverride _returnvalue_2 (Variables|Default|GetDBNO%20Override%20Tag)))
        (:Is%20Not%20Valid
          _))
      (CallFunction|CleanUpDBNOStatusTags)
      (if (Inputs|IsUsingInputType "Touch" (Fortnite|GetFortPlayerControllerfromActor (CallFunction|GetPawnOwner)))
        (Character|UnbindEventfromOnCharacterMovementUpdated (CallFunction|GetPawnOwner) (EventDispatchers|CreateEvent :!event OnMovementUpdated)))
      (elif (ScalableFloat|GetValueasBool (Variables|Default|GetApplyBlockDriverTagDirectly_Enabled))
        ;;[1 1 1] Remove "BlockDriverSeat" Tag
        (Ability|GameplayEffect|RemoveLooseGameplayTags FortPawn (GameplayTags|MakeGameplayTagContainerfromTag "(TagName=\"Athena.Gameplay.Status.BlockDriverSeat\")") true)))))

(event EventBeginPlay
  (bind _pawn (CallFunction|GetOwnerPawn))
  (bind _pawn_2 (CallFunction|GetOwnerPawn))
  ;[1 1 1] Begin Play
  (Utilities|IsValid _pawn_2
    (:Is%20Valid
      (Utilities|IsValid (Ability|GetAbilitySystemComponent _pawn_2)
        (:Is%20Valid
          (CallFunction|AddHotfixTags))
        (:Is%20Not%20Valid
          (EventDispatchers|AssignOnAbilitySystemComponentInitialized _pawn_2 (EventDispatchers|CreateEvent :!event Custom|OnAbilitySystemComponentInitialized_Event))))
      (CallFunction|OverrideThrowCarriedPlayerStrength (ScalableFloat|GetValueatLevel (Variables|Default|GetOverrideDBNOThrowStrengthXY)) (ScalableFloat|GetValueatLevel (Variables|Default|GetOverrideDBNOThrowStrengthZ)))
      (EventDispatchers|BindEventtoOnDBNOStateChanged _pawn (EventDispatchers|CreateEvent :!event Custom|On%20DBNO%20State%20Changed))
      (Vehicle|BindEventtoOnVehicleStateChangedEvent _pawn (EventDispatchers|CreateEvent :!event Custom|On%20Vehicle%20State%20Changed)))))

(event EventEndPlay
  (params (EndPlayReason :type EEndPlayReason%20Enum :default "Destroyed"))
  (bind _pawn (CallFunction|GetOwnerPawn))
  ;[1 1 1] End Play
  (|CallParentFunction EndPlayReason)
  (Utilities|Time|ClearandInvalidateTimerbyHandle (Variables|Default|GetSoundLibraryTimer))
  (bind _showholsteredweaponcomponent (Variables|Default|GetShowHolsteredWeaponComponent)
    (:then
      (Class|FortAthenaShowHolsteredWeaponsComponent|ClearHideHolsteredWeaponFunction _showholsteredweaponcomponent))
    (:else
      _))
  (Ability|GameplayEffect|RemoveLooseGameplayTags _pawn (GameplayTags|MakeGameplayTagContainerfromTag "(TagName=\"Gameplay.Action.Player.DBNOHurdleEnabled\")"))
  (EventDispatchers|UnbindEventfromOnDBNOStateChanged _pawn (EventDispatchers|CreateEvent :!event Custom|On%20DBNO%20State%20Changed))
  (Vehicle|UnbindEventfromOnVehicleStateChangedEvent _pawn (EventDispatchers|CreateEvent :!event Custom|On%20Vehicle%20State%20Changed)))

(event Custom|On%20Vehicle%20State%20Changed
  (params (FortPlayerPawn :type Fort%20Player%20Pawn%20Object%20Reference) (NewVehicle :type Actor%20Object%20Reference) (OldVehicle :type Actor%20Object%20Reference))
  (bind _returnvalue (Fortnite|GetFortPlayerControllerfromActor FortPlayerPawn))
  (bind _returnvalue_2 (PlayerController|LocalPlayerSubsystems|GetEnhancedInputLocalPlayerSubsystem _returnvalue))
  ;[1 1 1] On entering or exiting the vehicle, change the USE input context
  (if (Pawn|IsLocallyControlled FortPlayerPawn)
    (Utilities|IsValid _returnvalue
      (:Is%20Valid
        (Utilities|IsValid NewVehicle
          (:Is%20Valid
            (Input|RemoveMappingContext _returnvalue_2 "/TrendConcentrate/Inputs/IMC_DBNO_Use.IMC_DBNO_Use" "(bIgnoreAllPressedKeysUntilRelease=True,bForceImmediately=False,bNotifyUserSettings=False)"))
          (:Is%20Not%20Valid
            (if (Health|IsDBNO FortPlayerPawn)
              (Input|AddMappingContext _returnvalue_2 "/TrendConcentrate/Inputs/IMC_DBNO_Use.IMC_DBNO_Use" 20 "(bIgnoreAllPressedKeysUntilRelease=True,bForceImmediately=False,bNotifyUserSettings=False)"))))))))

(event Custom|AddHotfixTags
  ;[1 1 1] Add Hotfix Tags
  (if (ScalableFloat|GetValueasBool (Variables|Default|GetDBNOHurdleEnabled))
    (Ability|GameplayEffect|AddLooseGameplayTags (CallFunction|GetOwnerPawn) (GameplayTags|MakeGameplayTagContainerfromTag "(TagName=\"Gameplay.Action.Player.DBNOHurdleEnabled\")"))))

(event Custom|OnAbilitySystemComponentInitialized_Event
  (params (AbilitySystemComponent :type Fort%20Ability%20System%20Component%20Object%20Reference) (AffectedPawn :type Fort%20Player%20Pawn%20Object%20Reference))
  ;[1 1 1] Add Hotfix Tags
  (CallFunction|AddHotfixTags))

(event EnhancedInputActionIA_DBNO_Roll
  (params (ActionValue :type Boolean) (ElapsedSeconds :type Float%20%28double-precision%29) (TriggeredSeconds :type Float%20%28double-precision%29) (InputAction :type Input%20Action%20Object%20Reference :default "/TrendConcentrate/Inputs/IA_DBNO_Roll.IA_DBNO_Roll")))

(event Roll|EventExecuteRollFailGameplayCue
  ;[1 1 1] DBNO Roll
  (GameplayCue|ExecuteGameplayCue%28Burst%29Local (CallFunction|GetOwnerPawn) (Variables|Default|GetRollFailTag) (Variables|Default|GetCue%20Parameters)))

(event EnhancedInputActionIA_DBNO_Sprint
  (params (ActionValue :type Boolean) (ElapsedSeconds :type Float%20%28double-precision%29) (TriggeredSeconds :type Float%20%28double-precision%29) (InputAction :type Input%20Action%20Object%20Reference :default "/TrendConcentrate/Inputs/IA_DBNO_Sprint.IA_DBNO_Sprint"))
  (bind _returnvalue (CallFunction|IsSprintToggleable)))

(event Roll|EventSetBotRollDirection
  (params (InBotRollDirection :type Vector%20%28by%20ref%29 :default "0, 0, 0"))
  (Variables|FortTrendConcentratePawnComponent|SetBotRollDirection InBotRollDirection))

(event EventBP_OnPawnStartedWalking
  (if (and (Health|IsDBNO (CallFunction|GetPawnOwner)) (Inputs|IsUsingInputType "Touch" (Fortnite|GetFortPlayerControllerfromActor (CallFunction|GetPawnOwner))))
    (CallFunction|CheckRagdollStatus)))

(fn CanApplyHolster
  (params (HolsterId :type Name%20%28by%20ref%29 :default "None"))
  (return false))

(fn CheckRagdollStatus
  (bind _dbnoragdollslopethresholdfromroll (Variables|Ragdoll|GetDBNORagdollSlopeThresholdFromRoll))
  (bind _returnvalue (ScalableFloat|GetValueatLevel _dbnoragdollslopethresholdfromroll))
  (bind _pawn (CallFunction|GetOwnerPawn))
  (bind _charactermovement (Variables|/Script/Engine.Character|GetCharacterMovement _pawn))
  (bind _returnvalue_2 (Pawn|Components|CharacterMovement|GetCurrentAcceleration _charactermovement))
  (bind _returnvalue_3 (Math|Vector|VectorLength _returnvalue_2))
  (bind _returnvalue_4 (> _returnvalue_3 0))
  (bind _returnvalue_5 (Transformation|GetActorForwardVector _pawn))
  (bind _returnvalue_6 (select _returnvalue_4 _returnvalue_2 _returnvalue_5))
  (bind _returnvalue_7 (CallFunction|IsSlopeSteepEnoughtoRagdoll _returnvalue_6 _returnvalue))
  (bind _returnvalue_8 (Components|GetOwner))
  (bind _willragdollifrolltag (Variables|Default|GetWillRagdollIfRollTag))
  (if (GameplayTags|HasMatchingGameplayTag (CallFunction|GetOwnerPawn) "(TagName=\"Gameplay.Status.DBNO.TrendConcentrate.Ragdoll\")")
    _
    (elif (!= (Variables|Default|GetbPrevRagdollCheckResult) _returnvalue_7)
      (bind _output_get (Variables|Default|SetbPrevRagdollCheckResult _returnvalue_7))
      (if _output_get
        (Ability|GameplayEffect|AddLooseGameplayTags _returnvalue_8 _willragdollifrolltag)
        (else
          (Ability|GameplayEffect|RemoveLooseGameplayTags _returnvalue_8 _willragdollifrolltag))))))

(fn GetOwnerPawn
  (bind _asfort_player_pawn_athena (Utilities|Casting|CastToFortPlayerPawnAthena (Components|GetOwner)))
  (return _asfort_player_pawn_athena))

(fn OnMovementUpdated
  (params (DeltaSeconds :type Float%20%28single-precision%29) (OldLocation :type Vector) (OldVelocity :type Vector))
  ; This is a cosmetic check, no need to run every frame
  (if (not (> (Math|Integer64|%25%28integer64%29 (Utilities|GetFrameCount) 5) 0))
    (CallFunction|CheckRagdollStatus)))

(fn SetMappingContextsEnabled
  (params (FortPawn :type Fort%20Pawn%20Object%20Reference) (bSprint :type Boolean) (bRoll :type Boolean) (bUse :type Boolean))
  (bind _returnvalue (Fortnite|GetFortPlayerControllerfromActor FortPawn))
  (Utilities|IsValid _returnvalue
    (:Is%20Valid
      (Variables|Default|SetCached%20DBNO%20Player%20Controller _returnvalue))
    (:Is%20Not%20Valid
      _))
  (bind _cached_dbno_player_controller (Variables|Default|GetCached%20DBNO%20Player%20Controller)
    (:then
      (if (Variables|GetbSprint)
        (Input|AddMappingContext (PlayerController|LocalPlayerSubsystems|GetEnhancedInputLocalPlayerSubsystem _cached_dbno_player_controller) "/TrendConcentrate/Inputs/IMC_DBNO_Sprint.IMC_DBNO_Sprint" 20 "(bIgnoreAllPressedKeysUntilRelease=True,bForceImmediately=False,bNotifyUserSettings=False)")
        (else
          (Input|RemoveMappingContext (PlayerController|LocalPlayerSubsystems|GetEnhancedInputLocalPlayerSubsystem _cached_dbno_player_controller) "/TrendConcentrate/Inputs/IMC_DBNO_Sprint.IMC_DBNO_Sprint" "(bIgnoreAllPressedKeysUntilRelease=True,bForceImmediately=False,bNotifyUserSettings=False)")))
      (if (Variables|GetbRoll)
        (Input|AddMappingContext (PlayerController|LocalPlayerSubsystems|GetEnhancedInputLocalPlayerSubsystem _cached_dbno_player_controller) "/TrendConcentrate/Inputs/IMC_DBNO_Roll.IMC_DBNO_Roll" 20 "(bIgnoreAllPressedKeysUntilRelease=True,bForceImmediately=False,bNotifyUserSettings=False)")
        (else
          (Input|RemoveMappingContext (PlayerController|LocalPlayerSubsystems|GetEnhancedInputLocalPlayerSubsystem _cached_dbno_player_controller) "/TrendConcentrate/Inputs/IMC_DBNO_Roll.IMC_DBNO_Roll" "(bIgnoreAllPressedKeysUntilRelease=True,bForceImmediately=False,bNotifyUserSettings=False)")))
      (if (Variables|GetbUse)
        (Input|AddMappingContext (PlayerController|LocalPlayerSubsystems|GetEnhancedInputLocalPlayerSubsystem _cached_dbno_player_controller) "/TrendConcentrate/Inputs/IMC_DBNO_Use.IMC_DBNO_Use" 20 "(bIgnoreAllPressedKeysUntilRelease=True,bForceImmediately=False,bNotifyUserSettings=False)")
        (else
          (Input|RemoveMappingContext (PlayerController|LocalPlayerSubsystems|GetEnhancedInputLocalPlayerSubsystem _cached_dbno_player_controller) "/TrendConcentrate/Inputs/IMC_DBNO_Use.IMC_DBNO_Use" "(bIgnoreAllPressedKeysUntilRelease=True,bForceImmediately=False,bNotifyUserSettings=False)"))))
    (:else)))

(fn ShouldEnableInputMappings
  (bind _pawn (CallFunction|GetOwnerPawn))
  (bind _returnvalue (Actor|GetComponentbyClass _pawn "/TrendConcentrate/Components/B_DBNOBehaviorOverrideComponent_TrendConcentrate.B_DBNOBehaviorOverrideComponent_TrendConcentrate_C"))
  (bind (_boutallowsprint _boutallowroll) (DBNO|AreDBNOInputActionsEnabled _returnvalue))
  (Utilities|IsValid _returnvalue
    (:Is%20Valid
      (return _boutallowsprint _boutallowroll))
    (:Is%20Not%20Valid
      (return false false))))
